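# M_LimeRock_OuterCollision is an unlit translucent material in the LimeRock collision mesh folder, likely for visualizing outer collision geometry.
# It holds a constant red color and a scalar Opacity parameter with a low default, suggesting a see-through red tint, though no pin connections are shown.
# Ray-traced shadows and automatic usage-flag setting in the editor are both disabled.
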Material /LimeRock/Assets/Meshes/Collision/M_LimeRock_OuterCollision
graph {
 properties {
  bAutomaticallySetUsageInEditor=false
  bCastRayTracedShadows=false
  blendMode=BLEND_Translucent
  shadingModel=MSM_Unlit
  usedShadingModels=Unlit
 }
 nodes [
  Constant3Vector() {constant=<rgba 1 0 0 0>}
  ScalarParameter() {defaultValue=0.2, parameterName=Opacity}
 ]
}
